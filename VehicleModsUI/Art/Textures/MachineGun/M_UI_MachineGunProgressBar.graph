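# Translucent UI material that draws the machine gun overheat progress bar on the HUD, with fill, empty part, outline and optional glow.
# SmoothStep thresholds on one overheat texture derive bar shape and glow from red, and outline from green.
# AnimateProgress drives both a UV rotation and a ProgressStart-to-ProgressEnd offset, producing an edge mask with adjustable sharpness.
# Named reroutes carry these masks into lerp chains that layer color and opacity.
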Material /VehicleModsUI/Art/Textures/MachineGun/M_UI_MachineGunProgressBar
graph {
 properties {
  bAutomaticallySetUsageInEditor=false
  blendMode=BLEND_Translucent
  materialDomain=MD_UI
  pixelDepthOffsetMode=PDOM_Legacy
  usedShadingModels=Default Lit
 }
 nodes [
  n1 = NamedRerouteUsage()
  n2 = LinearInterpolate(A=LinearInterpolate(B=Multiply(A=@n1) {constB=0.3}, Alpha=ScalarParameter() {parameterName=EnableGlow}), Alpha=NamedRerouteUsage()) {constB=0}
  n3 = NamedRerouteUsage()
  n4 = NamedRerouteUsage()
  n5 = ScalarParameter() {defaultValue=0.5, parameterName=EmptyPartOpacity}
  n6 = NamedRerouteUsage()
  LinearInterpolate(A=LinearInterpolate(A=LinearInterpolate(A=@n2, B=@n3, Alpha=@n3), B=Multiply(A=@n4, B=@n5), Alpha=@n4), B=@n6, Alpha=@n6)
  LinearInterpolate(
   A=LinearInterpolate(A=VectorParameter() {parameterName=ProgressColor, defaultValue=<rgba 1 0 0.014 1>}.RGB, B=Constant3Vector(), Alpha=NamedRerouteUsage())
   B=VectorParameter() {parameterName=OutlineColor, defaultValue=<rgba 0.4401 0.4401 0.4401 1>}.RGB
   Alpha=NamedRerouteUsage()
  )
  n7 = TextureSample() {texture=&/T_UI_HUD_Xhair_Machinegun_overheat}
  NamedRerouteDeclaration(None=SmoothStep(Value=@n7.R) {constMax=0.348, constMin=0.32, constValue=0.232}) {name=ProgressBarShape, nodeColor=<rgba 0 0 0 1>}
  NamedRerouteDeclaration(None=SmoothStep(Value=TextureSample() {texture=&T_UI_HUD_Xhair_Machinegun_overheat}.G) {constMax=0.212, constMin=0.184, constValue=0.232}) {name=outline, nodeColor=<rgba 0 0 0 1>}
  n8 = NamedRerouteUsage()
  n9 = ScalarParameter() {defaultValue=1, parameterName=ProgressOpacity}
  NamedRerouteDeclaration(None=Multiply(A=NamedRerouteUsage(), B=Multiply(A=@n8, B=@n9))) {name=ProgressBar, nodeColor=<rgba 0 0 0 1>}
  n10 = OneMinus(None=NamedRerouteUsage())
  n11 = ScalarParameter() {defaultValue=1, parameterName=ProgressOpacity}
  NamedRerouteDeclaration(None=Multiply(A=NamedRerouteUsage(), B=Multiply(A=@n10, B=@n11))) {name="ProgressBar(1-x)", nodeColor=<rgba 0 0 0 1>}
  n12 = ScalarParameter() {group=Progress_Layout, parameterName=ProgressEdgeSharpness, sliderMax=1, sortPriority=2}
  n13 = ScalarParameter() {
   defaultValue=1
   desc="Property for animating progress value change (from ProgressStart to ProgressEnd)\r\n0 - 1"
   group=Progress
   parameterName=AnimateProgress
   sliderMax=1
   sortPriority=3
  }
  n14 = MaterialFunctionCall(UVs=TextureCoordinate(), "Rotation Angle (0-1)"=LinearInterpolate(Alpha=@n13) {constA=0.275, constB=0.525}) {materialFunction=&CustomRotator}
  n15 = ScalarParameter() {
   defaultValue=0.1
   desc="When progress value changes, it defines where change starts.\r\n0 - 1"
   group=Progress
   parameterName=ProgressStart
   sliderMax=1
   sortPriority=1
  }
  n16 = ScalarParameter() {
   defaultValue=0.9
   desc="When progress value changes, it defines where change ends.\r\n0 - 1"
   group=Progress
   parameterName=ProgressEnd
   sliderMax=1
   sortPriority=2
  }
  n17 = ScalarParameter() {
   defaultValue=1
   desc="Property for animating progress value change (from ProgressStart to ProgressEnd)\r\n0 - 1"
   group=Progress
   parameterName=AnimateProgress
   sliderMax=1
   sortPriority=3
  }
  n18 = LinearInterpolate(A=@n15, B=@n16, Alpha=@n17)
  n19 = ComponentMask(None=Add(A=Multiply(A=@n14.Rotated Values) {constB=-1}, B=NamedRerouteDeclaration(None=@n18) {name=Progress, nodeColor=<rgba 0 0 0 1>}) {constB=0.5}) {g=true}
  NamedRerouteDeclaration(None=Saturate(None=SmoothStep(Max=@n12, Value=@n19) {constMax=0.01})) {name=ProgressGraident, nodeColor=<rgba 0 0 0 1>}
  NamedRerouteDeclaration(None=SmoothStep(Max=ScalarParameter() {defaultValue=0.504, parameterName=GlowIntensity}, Value=@n7.R) {constMax=0.232, constValue=0.232}) {name=Glow, nodeColor=<rgba 0.3594 0 1 1>}
  SmoothStep() {constMax=0.232, constMin=0.204, constValue=0.232}
 ]
}
