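# Translucent UI material that procedurally draws a machine gun reticle for the VehicleMods UI, with no textures.
# Step functions on centered, mirrored UVs form a crosshair with adjustable thickness.
# Two MF_UI_Box SDFs with inner and outer size parameters form a band mask, and a distance field forms a small center dot.
# Named reroutes carry crosshair, mask and dot into a final multiply-add composite.
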Material /VehicleModsUI/Art/Textures/MachineGun/M_UI_VehicleMods_Reticle
graph {
 properties {blendMode=BLEND_Translucent, materialDomain=MD_UI, pixelDepthOffsetMode=PDOM_Legacy, usedShadingModels=Default Lit}
 nodes [
  Add(A=Multiply(A=NamedRerouteUsage(), B=NamedRerouteUsage()), B=NamedRerouteUsage())
  n1 = Abs(None=ConstantBiasScale(None=TextureCoordinate()) {bias=-0.5, scale=2})
  n2 = ComponentMask(None=@n1) {r=true}
  n3 = Divide(A=ScalarParameter() {defaultValue=0.25, parameterName=CrossHairThickness}) {constB=10}
  n4 = ComponentMask(None=@n1) {g=true}
  NamedRerouteDeclaration(None=Add(A=Step(Y=@n2, X=@n3) {constX=0.05}, B=Step(Y=@n4, X=@n3) {constX=0.05})) {name=CrossHair, nodeColor=<rgba 0 0 0 1>}
  n5 = MaterialFunctionCall(Size=ScalarParameter() {defaultValue=0.232, parameterName=MaskSizeInner}) {materialFunction=&MF_UI_Box}
  n6 = SmoothStep(Value=MaterialFunctionCall(Size=ScalarParameter() {defaultValue=0.768, parameterName=MaskSizeOutter}) {materialFunction=&MF_UI_Box}.SDF) {constMax=0.05}
  NamedRerouteDeclaration(None=Multiply(A=SmoothStep(Value=@n5.SDF) {constMax=0.05}, B=OneMinus(None=@n6))) {name=Mask, nodeColor=<rgba 0 0 0 1>}
  n7 = TextureCoordinate()
  n8 = Subtract(A=Distance(A=ConstantBiasScale(None=@n7) {bias=-0.5, scale=2}, B=Constant()), B=Constant() {r=0.02})
  NamedRerouteDeclaration(None=OneMinus(None=SmoothStep(Value=@n8) {constMax=0.02})) {name=MiddleCircle, nodeColor=<rgba 0 0 0 1>}
 ]
}
