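# Reticle extension widget for the vehicle-mod machine gun, showing a reticle, heat meter and disabled warning.
# Construct casts the equipped weapon, binds its overheat and targeting delegates, and initializes the meter; Destruct unbinds.
# Heat updates map current/max through a curve into the meter material, pulsing the reticle while heating and glowing the meter when overheated.
# Targeting plays a reticle mask animation forward or reverse, animating material parameters directly.
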WidgetBlueprint /VehicleModsUI/Widgets/MachineGun/WBP_VehicleMods_MachineGunReticle
animations [
 Anim_Targetting {
  binding_count=1
  length_seconds=0.25
  object=&self:Anim_Targetting
  sequence {bindings [Ritcle {
     id=#1
     possessed_class=Image
     tracks [{
       type=MovieSceneWidgetMaterialTrack
       sections [{
         type=MovieSceneParameterSection
         end_frame=5
         start_frame=0
         channels [
          MaskSizeInner {type=MovieSceneScriptingFloatChannel, keys [[0,0.232],[4.9997,0.132]]}
          MaskSizeOutter {type=MovieSceneScriptingFloatChannel, keys [[0,0.768],[4.9997,0.668]]}
         ]
        }]
      }]
    }]}
 }
 Anim_Fire {
  binding_count=1
  length_seconds=0.25
  object=&self:Anim_Fire
  sequence {bindings [Ritcle {
     id=#2
     possessed_class=Image
     tracks [{
       type=MovieSceneWidgetMaterialTrack
       sections [{
         type=MovieSceneParameterSection
         end_frame=5
         start_frame=0
         channels [
          MaskSizeInner {type=MovieSceneScriptingFloatChannel, keys [[0,0.232],[2,0.332],[5,0.232]]}
          MaskSizeOutter {type=MovieSceneScriptingFloatChannel, keys [[0,0.768],[2,0.868],[5,0.768]]}
         ]
        }]
      }]
    }]}
 }
 Anim_Overheated {
  binding_count=1
  length_seconds=2.5
  object=&self:Anim_Overheated
  sequence {bindings [HeatMeter {
     id=#3
     possessed_class=Image
     tracks [{
       type=MovieSceneWidgetMaterialTrack
       sections [{
         type=MovieSceneParameterSection
         end_frame=50
         start_frame=0
         channels [
          EnableGlow {type=MovieSceneScriptingFloatChannel, keys [[0,0],[4,1],[35,1],[50,0]]}
          GlowIntensity {type=MovieSceneScriptingFloatChannel, keys [[0,0.35],[4,0.25],[10,0.35],[35,0.424],[50,1]]}
         ]
        }]
      }]
    }]}
 }
]
graphs [
 EventGraph {nodes [
   Event Construct()
   > n1 = Cast To B_VehicleMachineGunModWeapon!(
    execute
    Object=GetEquippedWeapon!(self:Fort Weapon Reticle Extension Widget Base Object Reference) -> (ReturnValue:Fort Weapon Object Reference)
   ) -> (then, CastFailed, AsB Vehicle Machine Gun Mod Weapon:B Vehicle Machine Gun Mod Weapon Object Reference) {is=title K2Node_DynamicCast}
   > n2 = Set MachineGunRef(MachineGunRef:B Vehicle Machine Gun Mod Weapon Object Reference=@n1) -> (Output_Get:B Vehicle Machine Gun Mod Weapon Object Reference)
   > Bind Event to On Overheat Value Changed(self:Fort Weapon Ranged Object Reference=@n2, Delegate=Create Event()) {is=title K2Node_AddDelegate}
   > Bind Event to On Targeting Changed(
    self:Fort Weapon Object Reference=Get MachineGunRef() -> (MachineGunRef:B Vehicle Machine Gun Mod Weapon Object Reference)
    Delegate=Create Event()
   ) {is=title K2Node_AddDelegate}
   n3 = Get MachineGunRef()
   > Set Heat Progress Bar!(
    execute
    self
    CurrentValue:float=GetCurrentOverheatValue!(self:Fort Weapon Ranged Object Reference=@n3) -> (ReturnValue:float)
    MaxValue:float=GetOverheatingMaxValue!(self:Fort Weapon Ranged Object Reference=@n3) -> (ReturnValue:float)
    State:EFortWeaponOverheatState Enum=GetOverheatState!(self:Fort Weapon Ranged Object Reference=@n3) -> (ReturnValue:EFortWeaponOverheatState Enum=None)
    bIsInCooldown:bool=Get IsCooldown() -> (IsCooldown:bool)
   ) -> (then) {is=title K2Node_CallFunction}
   Event Destruct()
   n4 = Get MachineGunRef()
   > Unbind Event from On Overheat Value Changed(self:Fort Weapon Ranged Object Reference=@n4, Delegate=Create Event()) {is=title K2Node_RemoveDelegate}
   > Unbind Event from On Targeting Changed(self:Fort Weapon Object Reference=@n4, Delegate=Create Event()) {is=title K2Node_RemoveDelegate}
  ]}
 Set Heat Progress Bar {nodes [
   n1 = Set Heat Progress Bar!() -> (then, CurrentValue:float=~, MaxValue:float=~, State:EFortWeaponOverheatState Enum=None, bIsInCooldown:bool=~) {is=title K2Node_FunctionEntry}
   > n2 = Set OverheatPercent(OverheatPercent:double=float / float(A=@n1, B=@n1.MaxValue)) -> (Output_Get:double)
   > Set IsCooldown(IsCooldown:bool=Get bIsInCooldown() -> (bIsInCooldown:bool)) -> (Output_Get:bool)
   > n3 = GetDynamicMaterial(self=Get HeatMeter() -> (HeatMeter:Image Object Reference))
   > n4 = Set HeatMeterMaterialRef(HeatMeterMaterialRef:Material Instance Dynamic Object Reference=@n3) -> (Output_Get:Material Instance Dynamic Object Reference)
   > SetScalarParameterValue(
    self=@n4
    ParameterName=AnimateProgress
    Value=GetRuntimeFloatCurveValue!(
     Curve:"Runtime Float Curve Structure (by ref)"=Get OverheatCurve() -> (OverheatCurve:Runtime Float Curve Structure)
     InTime:float=@n2
     InDefaultValue:float=0
    ) -> (ReturnValue:float)
   )
   n5 = Select!(
    None:Widget Animation Object Reference
    Heating:Widget Animation Object Reference=Get Anim_Fire() -> (Anim_Fire:Widget Animation Object Reference)
    Cooling:Widget Animation Object Reference
    Overheated:Widget Animation Object Reference
    Index:EFortWeaponOverheatState Enum=@n1.State
   ) -> (ReturnValue:Widget Animation Object Reference)
   > StopAnimation(InAnimation=@n5)
   > PlayAnimationForward(InAnimation=@n5)
   > Branch(Condition=IsOverheated!(self:Fort Weapon Ranged Object Reference=Get MachineGunRef()) -> (ReturnValue:bool))
   > PlayAnimationForward(InAnimation=Get Anim_Overheated() -> (Anim_Overheated:Widget Animation Object Reference))
  ]}
 SetTargettingState {nodes [
   n1 = SetTargettingState!() -> (then, bIsTargeting:bool=~) {is=title K2Node_FunctionEntry}
   > n2 = Branch(Condition=@n1)
   n3 = Get Anim_Targetting() -> (Anim_Targetting:Widget Animation Object Reference)
   > PlayAnimationForward(InAnimation=@n3)
   PlayAnimationReverse(execute=@n2.else, InAnimation=@n3)
  ]}
]
variables [
 HeatMeterMaterialRef {type=MaterialInstanceDynamic}
 IsCooldown {type=boolean}
 MachineGunRef {type=B_VehicleMachineGunModWeapon_C, default=null}
 MaterialRef {type=MaterialInstanceDynamic}
 OverheatCurve {
  type=RuntimeFloatCurve
  default {
   externalCurve=&/OverheatCurve
   editorCurveData {defaultValue=3.4028234663852886e+38, postInfinityExtrap=RCCE_Constant, preInfinityExtrap=RCCE_Constant, keys []}
  }
 }
 OverheatPercent {type=number}
 State {type=EFortWeaponOverheatState, default=None}
]
widgets {
 info {
  inheritedWidgetCount=1
  parentClass=/Script/FortniteUI.FortWeaponReticleExtensionWidgetBase
  rootWidgetClass=&Overlay
  widgetCount=5
 }
 widgets [
  Root {widgetClassPath=&Overlay, widget=Root}
  Ritcle {
   widgetClassPath=&Image
   bIsVariable=true
   parent=Root
   widget=Ritcle
   properties {
    visibility=HitTestInvisible
    brush {
     bIsDynamicallyLoaded=false
     drawAs=Image
     imageType=FullColor
     mirroring=NoMirror
     resourceName=None
     resourceObject=/VehicleModsUI/Art/Textures/MachineGun/M_UI_VehicleMods_Reticle
     tiling=NoTile
     imageSize=<80 80>
     margin=<ltrb 0 0 0 0>
     outlineSettings {
      bUseBrushTransparency=false
      roundingType=HalfHeightRadius
      width=0
      color {colorUseRule=UseColor_Specified, specifiedColor=<rgba 0 0 0 0>}
      cornerRadii=<0 0 0 0>
     }
     tintColor {colorUseRule=UseColor_Specified, specifiedColor=<rgba 1 1 1 1>}
     uVRegion {bIsValid=false, max=<0 0>, min=<0 0>}
    }
   }
   slot {class=OverlaySlot, properties {horizontalAlignment=HAlign_Center, verticalAlignment=VAlign_Center}}
  }
  HeatMeter {
   widgetClassPath=&Image
   bIsVariable=true
   parent=Root
   widget=HeatMeter
   properties {
    brush {
     bIsDynamicallyLoaded=false
     drawAs=Image
     imageType=FullColor
     mirroring=NoMirror
     resourceName=None
     resourceObject=/VehicleModsUI/Art/Textures/MachineGun/M_UI_MachineGunProgressBar
     tiling=NoTile
     imageSize=<128 128>
     margin=<ltrb 0 0 0 0>
     outlineSettings {
      bUseBrushTransparency=false
      roundingType=HalfHeightRadius
      width=0
      color {colorUseRule=UseColor_Specified, specifiedColor=<rgba 0 0 0 0>}
      cornerRadii=<0 0 0 0>
     }
     tintColor {colorUseRule=UseColor_Specified, specifiedColor=<rgba 1 1 1 1>}
     uVRegion {bIsValid=false, max=<0 0>, min=<0 0>}
    }
    renderTransform {angle=0, scale=<1 1>, shear=<0 0>, translation=<75 44>}
   }
   slot {class=OverlaySlot}
  }
  WBP_VehicleMods_DisabledWarning {
   widgetClassPath=/VehicleModsUI/Widgets/WBP_VehicleMods_DisabledWarning
   bIsVariable=true
   parent=Root
   widget=WBP_VehicleMods_DisabledWarning
   slot {class=OverlaySlot, properties {horizontalAlignment=HAlign_Center, verticalAlignment=VAlign_Center}}
  }
  TutorialHighlight {widgetClassPath=&TutorialHighlight}
 ]
}
